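# Material function for Utopia club emissive surfaces, producing an animated swirling color flow with optional Orbitz branding.
# World-position planar UVs, oriented through custom primitive data, drive a distorted flow map whose grayscale result indexes a curve-atlas color ramp.
# A platform switch slightly desaturates the color outside high-end PC and current-gen consoles.
# Static switches add the logo using either bounding-box UVs or rotated world projection masked by UV and height.
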MaterialFunction /Hermes/Techart/UtopiaEmissiveSurfaces/Club/MF_ColorFlow
graph {nodes [
  n1 = Constant2Vector() {g=0.5, r=0.5}
  n2 = WorldPosition()
  n3 = LinearInterpolate(
   A=Constant3Vector() {constant=<rgba 1 1 0 1>}
   B=Constant3Vector() {constant=<rgba 1 -1 0 1>}
   Alpha=ScalarParameter() {bUseCustomPrimitiveData=true, parameterName=ColorFlow Orientation}
  )
  n4 = ScalarParameter() {bUseCustomPrimitiveData=true, parameterName=ColorFlow Orientation}
  n5 = LinearInterpolate(
   A=ComponentMask(None=MaterialFunctionCall!(VectorToRotate=@n2.XYZ, LookAtVector=@n3, RestingVector) {materialFunction=/Engine/Functions/Engine_MaterialFunctions01/Vectors/RotateVector}.Result) {b=true, r=true}
   B=ComponentMask(None=@n2.XYZ) {b=true, g=true}
   Alpha=Saturate(None=Subtract(A=@n4))
  )
  n6 = Divide(A=@n5, B=ScalarParameter() {defaultValue=1100, group=ColorFlow, parameterName=ColorFlow Size})
  n7 = Time()
  n8 = Multiply(A=@n7, B=ScalarParameter() {defaultValue=-0.003, group=ColorFlow, parameterName=ColorFlow Distortion Speed}) {constB=0.1}
  n9 = ComponentMask(None=TextureSample(UVs=Add(A=@n6, B=@n8)) {samplerType=SAMPLERTYPE_LinearColor, texture=/Game/Global/Textures/Default/Flowmaps/T_SwirlFlow_02}.RGB) {g=true, r=true}
  n10 = ScalarParameter() {defaultValue=0.25, group=ColorFlow, parameterName=ColorFlow FlowIntensity}
  n11 = ScalarParameter() {defaultValue=0.008, group=ColorFlow, parameterName=ColorFlow Speed}
  n12 = ComponentMask(None=MaterialFunctionCall!(
    Diffuse=TextureObject() {samplerType=SAMPLERTYPE_LinearGrayscale, texture=&/T_SwirlFlow_02_512_G8}
    Normal Map
    "Flow Vector Map (see tooltip)"=LinearInterpolate(A=@n1, B=@n9, Alpha=@n10)
    UV texture 2 offset
    "Time (see tooltip)"=Multiply(A=@n7, B=@n11) {constB=0.1}
    UVs=@n6
   ) {materialFunction=/Engine/Functions/Engine_MaterialFunctions02/Texturing/FlowMaps_Simple}.Diffuse) {r=true}
  n13 = ScalarParameter() {defaultValue=0.65, group=ColorFlow, parameterName=ColorFlow Pow}
  n14 = CurveAtlasRowParameter(CurveTime=SmoothStep(
    Min=ScalarParameter() {defaultValue=-0.15, group=ColorFlow, parameterName=ColorFlow SS_Min}
    Max=ScalarParameter() {defaultValue=1, group=ColorFlow, parameterName=ColorFlow SS_Max}
    Value=Power(Base=@n12, Exp=@n13)
   )) {
   atlas=&/CA_ColorFlow
   curve=&/C_ColorFlow2
   defaultValue=2
   parameterName=ColorCurve
  }
  n15 = Desaturation(None=@n14, Fraction=ScalarParameter() {defaultValue=0.05, parameterName=Desat_NotEpic})
  n16 = Multiply(
   A=MaterialFunctionCall!("Xb1,PS4"=@n15, "PC-High&&Epic,XSX,PS5"=@n14, "PC-Perf,Android"=@n15, Switch=@n15, "PC-Low(NotPerf)&&Medium"=@n15) {materialFunction=/Game/Effects/Niagara/MasterMaterials/Functions/Common/PlatformSwitch_3D}.Result
   B=ScalarParameter() {defaultValue=0.7, group=ColorFlow, parameterName=ColorFlow Intensity}
  )
  n17 = MaterialFunctionCall(Float3=Add(
    A=MaterialFunctionCall() {materialFunction=&BoundingBoxBased_0-1_UVW}.RGB
    B=VectorParameter() {parameterName=UVOffset, defaultValue=<rgba 0 0 0.04 1>}.RGB
   )) {materialFunction=&BreakOutFloat3Components}
  n18 = LinearInterpolate(A=OneMinus(None=@n17.R), B=@n17.R, Alpha=Ceil(None=Subtract(A=@n17.G) {constB=0.5}))
  n19 = OneMinus(None=@n17.B)
  n20 = VectorParameter() {parameterName="BrandingSize(UV)", defaultValue=<rgba 0.7 0.5 0 1>}
  n21 = TextureCoordinate()
  n22 = TextureSample!(
   UVs=MaterialFunctionCall(UVs=AppendVector(A=@n18, B=@n19), Texture Scale=ComponentMask(None=@n20.RGB) {g=true, r=true}) {materialFunction=&ScaleUVsByCenter}.UVs
   Tex
   "DDX(UVs)"=DDX(Value=@n21)
   "DDY(UVs)"=DDY(Value=@n21)
   Apply View MipBias
  ) {
   mipValueMode=TMVM_Derivative
   samplerSource=SSM_Clamp_WorldGroupSettings
   texture=&/T_Brand_Orbitz_TextOnly
  }
  n23 = VectorParameter() {parameterName=BrandingOffset2, defaultValue=<rgba -1.5676 1.2733 0 1>}
  n24 = VectorParameter() {parameterName=BrandingOffset, defaultValue=<rgba -1.4877 1.26 0 1>}
  n25 = ScalarParameter() {bUseCustomPrimitiveData=true, parameterName=ColorFlow Orientation}
  n26 = Add(
   A=Divide(A=@n5, B=ScalarParameter() {defaultValue=400, parameterName=BrandingSize}) {constB=1000}
   B=LinearInterpolate(A=ComponentMask(None=@n23.RGB) {g=true, r=true}, B=ComponentMask(None=@n24.RGB) {g=true, r=true}, Alpha=Saturate(None=@n25))
  )
  n27 = Multiply(
   A=MaterialFunctionCall(UVs=Multiply(A=@n26) {constB=-1}, "Rotation Angle (0-1)"=ScalarParameter() {defaultValue=0.25, parameterName=BrandingRotation}) {materialFunction=&CustomRotator}.Rotated Values
   B=ComponentMask(None=VectorParameter() {parameterName=BrandingScale, defaultValue=<rgba 0.75 1 0 0>}.RGB) {g=true, r=true}
  )
  n28 = MaterialFunctionCall(Float2=@n27) {materialFunction=&BreakOutFloat2Components}
  n29 = Multiply(A=@n28.R) {constB=-1}
  n30 = ScalarParameter() {bUseCustomPrimitiveData=true, parameterName=ColorFlow Orientation}
  n31 = TextureSample!(
   UVs=LinearInterpolate(
    A=MaterialFunctionCall(X=@n29, Y=@n28.G) {materialFunction=&MakeFloat2}.Result
    B=@n27
    Alpha=Saturate(None=@n30)
   )
   Tex
   Apply View MipBias
  ) {
   samplerSource=SSM_Wrap_WorldGroupSettings
   texture=/Hermes/Techart/UtopiaEmissiveSurfaces/Glass_4Layer/Base_Instances/ProjectorDFX/T_BrandOrbitz
  }
  n32 = MaterialFunctionCall(In=@n31.R, Contrast=ScalarParameter() {defaultValue=0.1, parameterName=BrandingContrast}) {materialFunction=&CheapContrast}
  n33 = ConstantBiasScale(None=ComponentMask(None=TextureCoordinate() {uTiling=0.5}) {r=true}) {bias=-0.5, scale=2}
  n34 = MaterialFunctionCall(
   In=Power(Base=Abs(None=@n33), Exp=ScalarParameter() {defaultValue=2.166, parameterName=BrandingMaskPow}) {constExponent=32}
   Contrast=Constant() {r=100}
  ) {materialFunction=&CheapContrast}
  n35 = Divide(A=Subtract(A=WorldPosition().Z) {constB=384}) {constB=250}
  n36 = Multiply(A=Saturate(None=OneMinus(None=@n34.Result)), B=Saturate(None=Floor(None=@n35)))
  n37 = Add(A=@n16, B=StaticSwitchParameter(True=@n22.R, False=Multiply(A=@n32.Result, B=@n36)) {parameterName=BoundingBoxBasedUV})
  n38 = ScalarParameter() {defaultValue=0.55, parameterName=BrandingOpacity}
  FunctionOutput(None=StaticSwitchParameter(True=LinearInterpolate(A=@n16, B=@n37, Alpha=@n38), False=@n16) {parameterName=Branded}) {outputName=Out}
 ]}
