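# Nightclub dancefloor material with animated color and dots, plus time-triggered warm-up and activation envelopes.
# MF_ColorFlow provides the base, brightened and saturation-shifted by sine rings radiating from the center and an expanding edge ring.
# Three BumpOffset dot layers at different depths, each tinted and center-faded, add fake parallax.
# They glow with the rings and also blend through a time-varying noise, modulated by polar-panned noise.
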Material /Hermes/Techart/UtopiaEmissiveSurfaces/Club/M_NightClub_Dancefloor
graph {
 properties {
  bAutomaticallySetUsageInEditor=false
  previewMesh=/Hermes/Sets/UtopiaCity/Meshes/SM_UtopiaCity_Club_DanceFloor_GPad_A_A
  usedShadingModels=Default Lit
 }
 nodes [
  n1 = Subtract(A=Distance(A=TextureCoordinate(), B=Constant() {r=0.5}), B=Time())
  n2 = Multiply(A=Saturate(None=Sine(None=@n1) {period=0.2}), B=NamedRerouteUsage())
  n3 = Multiply(
   A=MaterialFunctionCall() {materialFunction=&/MF_ColorFlow}.Out
   B=LinearInterpolate(
    A=ScalarParameter() {defaultValue=0.025, parameterName=BackgroundIntensity}
    B=ScalarParameter() {defaultValue=0.2, parameterName=BackgroundIntensityActive}
    Alpha=@n2
   )
  ) {constB=0.05}
  n4 = Desaturation(None=@n3, Fraction=ScalarParameter() {defaultValue=-1, parameterName=Saturation_Active})
  n5 = TextureCoordinate()
  n6 = Constant() {r=0.5}
  n7 = Multiply(A=Constant() {r=1}, B=NamedRerouteUsage())
  n8 = SphereMask(A=@n5, B=@n6, Radius=Subtract(A=@n7) {constB=0.05}) {attenuationRadius=0.4, hardnessPercent=80}
  n9 = Add(A=@n2, B=Multiply(A=SphereMask(A=@n5, B=@n6, Radius=@n7) {attenuationRadius=0.5, hardnessPercent=80}, B=OneMinus(None=@n8)))
  n10 = TextureCoordinate()
  n11 = BumpOffset(Coordinate=@n10, Height=Constant() {r=-1}, HeightRatioInput=ScalarParameter() {defaultValue=0.05, parameterName=Depth})
  n12 = SphereMask(A=@n11, B=Constant() {r=0.5}, Radius=ScalarParameter() {defaultValue=0.2, parameterName=CenterFade}) {attenuationRadius=0.2, hardnessPercent=50}
  n13 = BumpOffset(Coordinate=@n10, Height=Constant() {r=-1}, HeightRatioInput=ScalarParameter() {defaultValue=0.005, parameterName=Depth3})
  n14 = TextureSample(UVs=@n13) {samplerSource=SSM_Wrap_WorldGroupSettings, texture=&/T_Labs_Dots_Mask}
  n15 = Multiply(A=Multiply(A=OneMinus(None=@n12), B=Multiply(A=@n14.B)), B=VectorParameter() {parameterName=Color, defaultValue=<rgba 1 0.1 0.9106 1>}.RGB)
  n16 = BumpOffset(Coordinate=@n10, Height=Constant() {r=-1}, HeightRatioInput=ScalarParameter() {defaultValue=0.025, parameterName=Depth2})
  n17 = TextureSample(UVs=@n16) {samplerSource=SSM_Wrap_WorldGroupSettings, texture=&T_Labs_Dots_Mask}
  n18 = SphereMask(A=@n16, B=Constant() {r=0.5}, Radius=ScalarParameter() {defaultValue=0.35, parameterName=CenterFade2}) {attenuationRadius=0.2, hardnessPercent=50}
  n19 = Multiply(A=Multiply(A=Multiply(A=@n17.G), B=OneMinus(None=@n18)), B=VectorParameter() {parameterName=Color1, defaultValue=<rgba 0 0.2672 1 1>}.RGB)
  n20 = TextureSample(UVs=@n11) {samplerSource=SSM_Wrap_WorldGroupSettings, texture=&T_Labs_Dots_Mask}
  n21 = SphereMask(A=@n13, B=Constant() {r=0.5}, Radius=ScalarParameter() {defaultValue=0.35, parameterName=CenterFade3}) {attenuationRadius=0.2, hardnessPercent=50}
  n22 = Multiply(A=Multiply(A=Multiply(A=@n20.R), B=OneMinus(None=@n21)), B=VectorParameter() {parameterName=Color2, defaultValue=<rgba 1 0.0187 0 1>}.RGB)
  n23 = Power(Base=Max(A=Max(A=@n15, B=@n19), B=@n22), Exp=ScalarParameter() {defaultValue=2, parameterName=DotsPow})
  n24 = Multiply(A=@n9) {constB=4}
  n25 = TextureSample() {
   samplerSource=SSM_Wrap_WorldGroupSettings
   samplerType=SAMPLERTYPE_Grayscale
   texture=/Game/Environments/Apollo/Sets/IceHotel/Textures/RefractionNoise
  }
  n26 = Sine(None=Multiply(A=Time()) {constB=0.1})
  n27 = Add(A=Multiply(A=ConstantBiasScale(None=@n26)) {constB=4}) {constB=-1}
  n28 = Clamp(None=Saturate(None=Power(Base=@n25.R, Exp=@n27))) {maxDefault=0.85, minDefault=0.15}
  n29 = MaterialFunctionCall!(
   UV Coord=TextureCoordinate()
   LenghtScale=ScalarParameter() {defaultValue=2, parameterName=Noise_PolarLength}
   RadialScale=ScalarParameter() {defaultValue=0.5, parameterName=Noise_PolarScale}
  ) {materialFunction=/Game/Global/MaterialFunctions/TechArt/MF_PolarCoordinates}
  n30 = Multiply(A=Time(), B=ScalarParameter() {defaultValue=1, parameterName=Noise_Speed})
  n31 = LinearInterpolate(
   A=ScalarParameter() {parameterName=Noise_Low}
   B=ScalarParameter() {defaultValue=3, parameterName=Noise_High}
   Alpha=TextureSample(UVs=Panner(Coordinate=@n29.Result, Time=@n30) {speedX=0.5}) {samplerSource=SSM_Wrap_WorldGroupSettings, texture=/Game/Effects/Fort_Effects/Textures/Tile/T_GaussianNoise_01_M}.G
  )
  Add(
   A=Add(A=LinearInterpolate(A=@n3, B=@n4, Alpha=@n9), B=Multiply(A=@n23, B=@n24))
   B=Multiply(
    A=MaterialFunctionCall(A=@n15, B=@n19, C=@n22, Alpha=@n28) {materialFunction=&3ColorBlend}.Result
    B=Saturate(None=@n31)
   )
  )
  Constant() {r=1}
  Constant() {r=0.05}
  Constant() {r=0.1}
  n32 = Divide(
   A=Subtract(A=Time(), B=ScalarParameter() {parameterName=TriggerTime_Warmup})
   B=ScalarParameter() {defaultValue=2.5, parameterName=TriggerDuration_Warmup}
  ) {constB=3}
  NamedRerouteDeclaration(None=OneMinus(None=Saturate(None=@n32))) {name=WarmUpAnim, nodeColor=<rgba 0.3125 0 1 1>}
  n33 = Constant() {r=-0.5}
  n34 = Divide(
   A=Subtract(A=Time() {period=3600}, B=ScalarParameter() {parameterName=TriggerTime})
   B=ScalarParameter() {defaultValue=5, parameterName=TriggerDuration}
  )
  n35 = Frac(None=@n34)
  n36 = Multiply(A=Abs(None=Add(A=@n33, B=@n35))) {constB=2}
  n37 = Subtract(A=OneMinus(None=Saturate(None=@n36)), B=Saturate(None=Floor(None=@n34)))
  NamedRerouteDeclaration(None=Power(Base=Saturate(None=@n37)) {constExponent=0.2}) {name=ActivationAnim, nodeColor=<rgba 0.0078 1 0 1>}
 ]
}
