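# Material function that outputs a pixel-pattern mask, sited among the club emissive surface assets, which fades out with camera distance.
# A static switch selects cylindrical UVs, built from the angle around the object pivot and world height, or tiled mesh UVs with bump-offset depth.
# The saturated length of the sampled RGB pattern blends against white, strongest near the camera up to a maximum occlusion and returning to 1 far away.
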MaterialFunction /Hermes/Techart/UtopiaEmissiveSurfaces/Club/MF_PixelPattern
graph {nodes [
  n1 = WorldPosition() {worldPositionShaderOffset=WPT_ExcludeAllShaderOffsets}
  n2 = ComponentMask(None=Subtract(A=@n1.XYZ, B=ObjectPositionWS())) {g=true, r=true}
  n3 = Divide(A=@n1.XYZ, B=Constant() {r=50})
  n4 = MaterialFunctionCall(Float2=AppendVector(
    A=MaterialFunctionCall(Vector or UVs=@n2) {materialFunction=&VectorToRadialValue}.Vector Converted to Angle
    B=ComponentMask(None=@n3) {b=true}
   )) {materialFunction=&BreakOutFloat2Components}
  n5 = VectorParameter() {parameterName=CylindricalUVsTiling, defaultValue=<rgba 48 3 0 1>}
  n6 = MaterialFunctionCall(X=Multiply(A=@n4.R, B=@n5.R) {constB=48}, Y=Multiply(A=@n4.G, B=@n5.G) {constB=3}) {materialFunction=&MakeFloat2}
  n7 = ScalarParameter() {defaultValue=24, group=PixelPattern, parameterName=PixelPattern_Tiling_X}
  n8 = ScalarParameter() {defaultValue=24, group=PixelPattern, parameterName=PixelPattern_Tiling_Y}
  n9 = BumpOffset(
   Coordinate=Multiply(A=TextureCoordinate(), B=AppendVector(A=@n7, B=@n8))
   Height=Constant() {r=1}
   HeightRatioInput=ScalarParameter() {defaultValue=-1, group=PixelPattern, parameterName=PixelPattern_Depth}
  )
  n10 = Length(None=TextureSample(UVs=StaticSwitchParameter(True=@n6.Result, False=@n9) {parameterName=UseCylindricalUVs}) {samplerSource=SSM_Wrap_WorldGroupSettings, texture=&/T_PixelPattern_RGB}.RGB) {is=class MaterialExpressionLength}
  n11 = WorldPosition() {worldPositionShaderOffset=WPT_CameraRelative}
  n12 = Max(A=Subtract(A=Length(None=@n11.XYZ), B=ScalarParameter() {defaultValue=-350, group=PixelPattern, parameterName=PixelPattern_DistanceFade_Offset}) {constB=250}) {constB=0}
  n13 = ScalarParameter() {defaultValue=1000, group=PixelPattern, parameterName=PixelPattern_DistanceFade_Range}
  n14 = OneMinus(None=Saturate(None=Divide(A=@n12, B=@n13) {constB=1000}))
  n15 = ScalarParameter() {defaultValue=0.25, group=PixelPattern, parameterName=PixelPattern_DistanceFade_MaxOcclusion}
  FunctionOutput(None=LinearInterpolate(B=Saturate(None=@n10), Alpha=Min(A=@n14, B=@n15)) {constA=1}) {outputName=Out}
 ]}
