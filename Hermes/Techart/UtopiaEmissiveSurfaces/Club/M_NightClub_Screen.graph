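# Default Lit material for a nightclub screen surface, with a dark, glossy metallic base and animated-looking colored emissive.
# MF_ColorFlow times MF_PixelPattern feeds EyeAdaptationInverse to form the visible emissive.
# A RayTracingQualitySwitch gives ray tracing the unpatterned color flow scaled by LumenIntensity instead.
# A platform switch halves base color on Switch, Android and PC performance mode, and the final attributes pass through MF_FN_DestructibleDamage.
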Material /Hermes/Techart/UtopiaEmissiveSurfaces/Club/M_NightClub_Screen
graph {
 properties {bAutomaticallySetUsageInEditor=false, bUseMaterialAttributes=true, usedShadingModels=Default Lit}
 nodes [
  n1 = ScalarParameter() {defaultValue=0.08, parameterName=BaseColor}
  n2 = Multiply(A=@n1) {constB=0.5}
  n3 = MaterialFunctionCall() {materialFunction=&/MF_ColorFlow}
  n4 = EyeAdaptationInverse(LightValueInput=Multiply(A=MaterialFunctionCall() {materialFunction=&/MF_PixelPattern}.Out, B=@n3.Out))
  n5 = Multiply(A=@n3.Out, B=ScalarParameter() {defaultValue=1, group=Lighting, parameterName=LumenIntensity})
  MaterialFunctionCall(Material_In=MakeMaterialAttributes!(
    BaseColor=MaterialFunctionCall("Xb1,PS4"=@n1, "PC-High&&Epic,XSX,PS5"=@n1, "PC-Perf,Android"=@n2, Switch=@n2, "PC-Low(NotPerf)&&Medium"=@n1) {materialFunction=&PlatformSwitch_1D}.Result
    Metallic=Constant() {r=1}
    Specular=ScalarParameter() {defaultValue=0.5, parameterName=Specular}
    Roughness=Constant() {r=0.1}
    Anisotropy
    EmissiveColor=RayTracingQualitySwitch(Normal=@n4.EyeAdaptationInverse, RayTraced=@n5)
    Opacity
    OpacityMask
    Normal
    Tangent
    WorldPositionOffset
    SubsurfaceColor
    ClearCoat
    ClearCoatRoughness
    AmbientOcclusion
    Refraction
    CustomizedUVs_0
    CustomizedUVs_1
    CustomizedUVs_2
    CustomizedUVs_3
    CustomizedUVs_4
    CustomizedUVs_5
    CustomizedUVs_6
    CustomizedUVs_7
    PixelDepthOffset
    ShadingModel
    Displacement
   )) {materialFunction=&MF_FN_DestructibleDamage}
 ]
}
